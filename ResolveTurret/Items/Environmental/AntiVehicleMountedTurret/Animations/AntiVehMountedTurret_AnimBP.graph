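# Animation Blueprint for the mounted anti-vehicle turret, posing it by driver presence and aim.
# A bool blend picks the default pose with a driver or the off pose without, then feeds DefaultSlot, a pitch aim offset and a lean aim offset driven by smoothed yaw delta.
# A bone modifier sets the base bone's world-space yaw to AimingYaw before inertialization.
# The event graph is empty, and the three declared "_old" variables are unused.
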AnimBlueprint /ResolveTurret/Items/Environmental/AntiVehicleMountedTurret/Animations/AntiVehMountedTurret_AnimBP
target_skeleton=/ResolveTurret/Items/Environmental/AntiVehicleMountedTurret/Meshes/Anti_Vehicle_Mounted_Turret_Gadget_Skeleton
graphs [
 AnimGraph {nodes [
   n1 = Sequence Player 'AVMT_Off_Pose_Weapon'() {sequence=&/AVMT_Off_Pose_Weapon}
   n2 = Sequence Player 'AntiVehMountedTurret_DefaultPose_Weapon'() {sequence=&/AntiVehMountedTurret_DefaultPose_Weapon}
   n3 = Get bHasDriver() -> (bHasDriver:bool)
   n4 = AimOffset Player 'AntiVehMountedTurret_AO_Weapon'!(
    X:float=NormalizeAxis!(Angle:float=Get AimingPitch() -> (AimingPitch:float)) -> (ReturnValue:float)
    BasePose:Pose Link Structure=Slot 'DefaultSlot'(Source=Blend Poses by bool(BlendPose_0=@n2, BlendPose_1=@n1, bActiveValue=@n3) {blendType=Linear})
    Alpha:float=1
   ) -> (Pose:Pose Link Structure) {
    blendSpace=&/AntiVehMountedTurret_AO_Weapon
    alphaBoolBlend {blendOption=Linear}
   }
   n5 = AimOffset Player 'AVMT_LeanAO_Weapon'!(X:float=Get YawDeltaSmoothed() -> (YawDeltaSmoothed:float), BasePose:Pose Link Structure=@n4, Alpha:float=1) -> (Pose:Pose Link Structure) {blendSpace=&/AVMT_LeanAO_Weapon, alphaBoolBlend {blendOption=Linear}}
   n6 = "Transform (Modify) Bone - Bone: base"!(
    ComponentPose=Local To Component(LocalPose=@n5)
    Alpha=1
    Rotation_Roll:float=0
    Rotation_Pitch:float=0
    Rotation_Yaw:float=Get AimingYaw() -> (AimingYaw:float)
   ) -> (Pose) {rotationMode=BMM_Replace, rotationSpace=BCS_WorldSpace, alphaBoolBlend {blendOption=Linear}, boneToModify {boneName=base}}
   Output Pose(Result=Inertialization!(Source:Pose Link Structure=Component To Local(ComponentPose=@n6)) -> (Pose:Pose Link Structure) {is=title AnimGraphNode_Inertialization}) {layerGroup=None}
  ]}
 EventGraph {links [], nodes []}
]
variables [
 DriverPawn_old {type=FortPlayerPawn}
 HasDriver_old {type=boolean}
 YawDeltaSmoothed_old {type=number}
]
