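# Physics asset for the anti-vehicle mounted turret's skeletal mesh, providing its per-bone collision and joint setup.
# Each of the eight bodies is a single box, covering the offset and base, three gun arm segments, body, barrel and shield.
# Constraints chain offset through base and arm to gun_arm_end, which branches to the shield and to the body with its barrel.
# All joints share identical swing and twist limit angles.
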PhysicsAsset /ResolveTurret/Items/Environmental/AntiVehicleMountedTurret/Meshes/Anti_Vehicle_Mounted_Turret_Physic
bodies {
 base [{shapeType=Box, extentX=114.5912, extentY=72.4558, extentZ=8.9002, center=<-33.68 0 6.3774>}]
 gun_arm_end [{shapeType=Box, extentX=52.6226, extentY=13.5323, extentZ=37.8906, center=<-19.3399 0 12.047>}]
 gun_arm_mid [{shapeType=Box, extentX=23.1549, extentY=15.4815, extentZ=42.5691, center=<10.7375 0 13.2305>, rotation=<rot -40 0.0002 -0.0004>}]
 gun_arm_start [{shapeType=Box, extentX=31.3083, extentY=22.1569, extentZ=46.927, center=<5.5447 0 15.575>}]
 gun_barrel_02 [{shapeType=Box, extentX=65.717, extentY=22.9084, extentZ=33.4567, center=<10.289 0 0.0087>}]
 gun_body [{shapeType=Box, extentX=81.459, extentY=81.8706, extentZ=46.8354, center=<3.7543 0.0809 -5.7532>}]
 gun_shield [{shapeType=Box, extentX=12.7386, extentY=75.9325, extentZ=66.5704, center=<0.087 0 -20.5388>}]
 offset [{shapeType=Box, extentX=124.8104, extentY=124.8104, extentZ=9.8955, center=<-0.128 0 4.8309>}]
}
constraints [
 {bone1Name=base, bone2Name=offset, swing1LimitDegrees=45, swing2LimitDegrees=45, twistLimitDegrees=45}
 {bone1Name=gun_arm_start, bone2Name=base, swing1LimitDegrees=45, swing2LimitDegrees=45, twistLimitDegrees=45}
 {bone1Name=gun_arm_mid, bone2Name=gun_arm_start, swing1LimitDegrees=45, swing2LimitDegrees=45, twistLimitDegrees=45}
 {bone1Name=gun_arm_end, bone2Name=gun_arm_mid, swing1LimitDegrees=45, swing2LimitDegrees=45, twistLimitDegrees=45}
 {bone1Name=gun_body, bone2Name=gun_arm_end, swing1LimitDegrees=45, swing2LimitDegrees=45, twistLimitDegrees=45}
 {bone1Name=gun_barrel_02, bone2Name=gun_body, swing1LimitDegrees=45, swing2LimitDegrees=45, twistLimitDegrees=45}
 {bone1Name=gun_shield, bone2Name=gun_arm_end, swing1LimitDegrees=45, swing2LimitDegrees=45, twistLimitDegrees=45}
]
